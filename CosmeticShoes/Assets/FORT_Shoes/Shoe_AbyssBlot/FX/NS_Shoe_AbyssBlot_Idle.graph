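# Niagara system for the AbyssBlot shoe cosmetic that draws short-lived anim-trail ribbons on both feet while running, despite its Idle name.
# Two ribbon emitters each spawn two skeletal trails, with width following per-shoe scale and alpha zeroed when not running or skydiving.
# One trail is tinted near-black.
# The other drives color, highlight, rim light and material parameters from time-of-day curves, with fixed overrides for front-end preview.
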NiagaraSystem /CosmeticShoes/Assets/FORT_Shoes/Shoe_AbyssBlot/FX/NS_Shoe_AbyssBlot_Idle
dependencies {
 usedDataInterfaces [&NiagaraDataInterfaceColorCurve, &NiagaraDataInterfaceCurve, &NiagaraDataInterfaceSkeletalMesh]
 usedDynamicInputs [/Game/Effects/Niagara/DynamicInputs/Logic/FN_IsRunningCheck, &Boolean_OrOperation]
 usedModules [
  &self:SystemSpawnScript.SystemScriptSource.SystemScriptGraph.NiagaraNodeAssignment_0.SetVariables_#1
  &SystemState
  &self:SystemSpawnScript.SystemScriptSource.SystemScriptGraph.NiagaraNodeAssignment_1.SetVariables_#2
 ]
}
emitters {
 FN_AnimTrail {
  emitterSpawnScript [&FN_RibbonIDOffsetUtility()]
  emitterUpdateScript [
   &EmitterState()
   &FN_SpawnAnimTrail(
    Number of Anim Trails:NiagaraInt32=&MakeCustomIntFromBool(Bool=$System.bIsRunning, True INT=2, False INT=0)
    Spawn Mode:ENiagaraSkeletalSpawning=::NewEnumerator0
    Switch on ENiagara_AnimTrail_Source:ENiagara_AnimTrail_Source=::NewEnumerator1
   )
  ]
  particleSpawnScript [
   &FN_InitializeAnimTrailBoneIndecies(Second Bone or Socket Index=0)
   &FN_InitializeAnimTrailBoneIndecies(AnimTrail Selection=1, First Bone or Socket Index=1) {module=FN_InitializeAnimTrailBoneIndecies001}
   &FN_InitializeAnimTrail(
    AnimTrailWidth=&Multiply_Float(
     A=25
     B=&MaskFloatBySpawnGroup(
      Masked Float=&MakeFloatFromVector(Vector=$User.ShoeScale_L, Channel=::NewEnumerator0)
      Passthrough Float=&MakeFloatFromVector(Vector=$User.ShoeScale_R, Channel=::NewEnumerator0)
     )
    )
    Color=<rgba 1 1 1 4>
    Lifetime=0.15
   )
  ]
  particleUpdateScript [
   &SolveForcesAndVelocity(Clamp Velocity=false, Limit Acceleration=false)
   &FN_UpdateAnimTrail() {module=FN_UpdateAnimTrail001}
   &ScaleColor(
    Linear Color Curve=DataInterface() {propertyValues {
      AlphaCurve {keys [<at 0.0144 1.2>,<at 0.1769 0.6>,<at 1 0>]}
      BlueCurve {keys [<at 0 0.0156>]}
      GreenCurve {keys [<at 0 0.0156>]}
      RedCurve {keys [<at 0 0.0156>]}
     }}
    Scale Mode=::NewEnumerator2
   )
   &DynamicMaterialParameters(
    Index 0 Param 1=0.85
    Index 0 Param 2=1
    Index 0 Param 3=-0.5
    Index 0 Param 4=&MakeCustomFloatFromBool(
     Bool=&Boolean_AndOperation(Bool1=$System.bIsRunning, Bool2=&Boolean_NotOperation(Bool=$User.IsSkydiving))
     True Float=1
     False Float=0
    )
    Module.Index0Float0Write=true
    Module.Index0Float1Write=true
    Module.Index0Float2Write=true
    Module.Index0Float3Write=true
   )
   &ScaleColor(
    Scale Alpha=&MakeCustomFloatFromBool(
     Bool=&Boolean_AndOperation(Bool1=$System.bIsRunning, Bool2=&Boolean_NotOperation(Bool=$User.IsSkydiving))
     True Float=1
     False Float=0
    )
    ScaleA=true
    ScaleRGB=true
   ) {module=ScaleColor001}
   &ScaleRibbonWidth(Ribbon Width Scale=&FloatFromCurve(FloatCurve=DataInterface() {propertyValues {Curve {keys [<at 0 1 RCIM_Cubic>, {arriveTangent=-0.6839, tangentMode=RCTM_User, time=1, value=0.2943}]}}}))
  ]
  rendererClasses [&NiagaraRibbonRendererProperties]
  renderers [{rendererClass=&NiagaraRibbonRendererProperties}]
 }
 FN_AnimTrail003 {
  emitterSpawnScript [&FN_RibbonIDOffsetUtility()]
  emitterUpdateScript [
   &EmitterState()
   &FN_SpawnAnimTrail(
    Number of Anim Trails:NiagaraInt32=&MakeCustomIntFromBool(Bool=$System.bIsRunning, True INT=2, False INT=0)
    Spawn Mode:ENiagaraSkeletalSpawning=::NewEnumerator0
    Switch on ENiagara_AnimTrail_Source:ENiagara_AnimTrail_Source=::NewEnumerator1
   )
  ]
  particleSpawnScript [
   &FN_InitializeAnimTrailBoneIndecies(Second Bone or Socket Index=0)
   &FN_InitializeAnimTrailBoneIndecies(AnimTrail Selection=1, First Bone or Socket Index=1) {module=FN_InitializeAnimTrailBoneIndecies001}
   &FN_InitializeAnimTrail(
    AnimTrailWidth=&Multiply_Float(
     A=22
     B=&MaskFloatBySpawnGroup(
      Masked Float=&MakeFloatFromVector(Vector=$User.ShoeScale_L, Channel=::NewEnumerator0)
      Passthrough Float=&MakeFloatFromVector(Vector=$User.ShoeScale_R, Channel=::NewEnumerator0)
     )
    )
    Color=<rgba 1 1 1 4>
    Lifetime=0.12
   )
  ]
  particleUpdateScript [
   &SolveForcesAndVelocity(Clamp Velocity=false, Limit Acceleration=false)
   &FN_UpdateAnimTrail() {module=FN_UpdateAnimTrail001}
   &ScaleColor(
    Scale Alpha=&FloatFromCurve(FloatCurve=DataInterface() {propertyValues {Curve {keys [
         <at 0.2159 -0.055 RCIM_Cubic>
         {arriveTangent=-0.887, interpMode=RCIM_Cubic, leaveTangent=-0.887, time=0.752, value=-0.0875}
         {arriveTangent=-1.5031, tangentMode=RCTM_User, time=1, value=-0.7505}
        ]}}})
    Scale RGB=&MakeVectorFromLinearColorRGB(Color=&LerpLinearColors(
      StartColor=&MakeCustomLinearColorFromBool(Bool=$User.bIsFrontEndPreview, True Color=<rgba 0.43 0.43 0.43 9>, False Color=<rgba 1 1 1 9>)
      EndColor=&MakeCustomLinearColorFromBool(
       Bool=$User.bIsFrontEndPreview
       True Color=<rgba 0.3229 0.5355 1 1>
       False Color=&ColorFromCurve(
        ColorCurve=DataInterface() {propertyValues {
          AlphaCurve {keys [<at 0 1>,<at 0.125 1>,<at 0.25 1>,<at 0.375 1>,<at 0.5 1>,<at 0.625 1>,<at 0.75 1>,<at 0.875 1>,<at 1 1>]}
          BlueCurve {keys [<at 0 0.28>,<at 0.125 0.28>,<at 0.1865 0.2671>,<at 0.3174 0.266>,<at 0.4889 0.266>,<at 0.791 0.266>,<at 0.875 0.224>,<at 0.916 0.28>,<at 1 0.28>]}
          GreenCurve {keys [<at 0 0.2404>,<at 0.125 0.2422>,<at 0.1865 0.266>,<at 0.3174 0.2674>,<at 0.4889 0.2687>,<at 0.791 0.2675>,<at 0.875 0.2487>,<at 0.916 0.2406>,<at 1 0.2404>]}
          RedCurve {keys [<at 0 0.196>,<at 0.125 0.196>,<at 0.1865 0.28>,<at 0.3174 0.28>,<at 0.4889 0.28>,<at 0.791 0.28>,<at 0.875 0.28>,<at 0.916 0.196>,<at 1 0.196>]}
         }}
        CurveIndex=Linked() {linkedVariable {name=NPC.FortniteNPC.TimeOfDay_0-1, type=NiagaraFloat}}
       )
      )
      LerpFactor=0
     ))
    ScaleA=true
    ScaleRGB=true
   )
   &ScaleColor(
    Scale Alpha=&MakeCustomFloatFromBool(
     Bool=&Boolean_AndOperation(Bool1=$System.bIsRunning, Bool2=&Boolean_NotOperation(Bool=$User.IsSkydiving))
     True Float=1
     False Float=0
    )
    ScaleA=true
    ScaleRGB=true
   ) {module=ScaleColor001}
   &ScaleRibbonWidth(Ribbon Width Scale=&FloatFromCurve(FloatCurve=DataInterface() {propertyValues {Curve {keys [
         {arriveTangent=-3.0003, interpMode=RCIM_Cubic, leaveTangent=-3.0003, tangentMode=RCTM_User, time=0.3113, value=1}
         {arriveTangent=-0.962, tangentMode=RCTM_User, time=1, value=0}
        ]}}}))
   /SetVariables(Particles.HighlightDynamic:Vector4f=&MakeVector4_FromLinearColor(LinearColor=&LerpLinearColors(
      StartColor=&ColorFromCurve(
       ColorCurve=DataInterface() {propertyValues {
         AlphaCurve {keys [<at 0 1>,<at 0.125 1>,<at 0.25 1>,<at 0.375 1>,<at 0.5 1>,<at 0.625 1>,<at 0.75 1>,<at 0.875 1>,<at 1 1>]}
         BlueCurve {keys [<at 0 0.9216>,<at 0.041 0.8963>,<at 0.083 0.8308>,<at 0.125 0.7835>,<at 0.1865 0.2611>,<at 0.25 0.3013>,<at 0.333 0.3128>,<at 0.4094 0.2806>,<at 0.4889 0.2616>,<at 0.583 0.2814>,<at 0.666 0.2801>,<at 0.75 0.2471>,<at 0.791 0.2118>,<at 0.833 0.2338>,<at 0.875 0.2473>,<at 0.916 0.8388>,<at 0.958 0.9047>,<at 1 0.9216>]}
         GreenCurve {keys [<at 0 0.5616>,<at 0.041 0.5276>,<at 0.083 0.4795>,<at 0.125 0.5099>,<at 0.1865 0.2253>,<at 0.25 0.3543>,<at 0.333 0.4809>,<at 0.4094 0.4334>,<at 0.4889 0.389>,<at 0.583 0.2814>,<at 0.666 0.4785>,<at 0.75 0.2471>,<at 0.791 0.2686>,<at 0.833 0.4442>,<at 0.875 0.5188>,<at 0.916 0.5292>,<at 0.958 0.5688>,<at 1 0.6308>]}
         RedCurve {keys [<at 0 0.1584>,<at 0.041 0.1774>,<at 0.083 0.1644>,<at 0.125 0.1755>,<at 0.1865 0.6867>,<at 0.25 0.8148>,<at 0.333 0.8963>,<at 0.4094 0.9131>,<at 0.4889 0.9131>,<at 0.583 0.8714>,<at 0.666 0.8963>,<at 0.75 0.7529>,<at 0.791 0.7529>,<at 0.833 0.8469>,<at 0.875 0.8632>,<at 0.916 0.1791>,<at 0.958 0.1696>,<at 1 0.305>]}
        }}
       CurveIndex=Linked() {linkedVariable {name=NPC.FortniteNPC.TimeOfDay_0-1, type=NiagaraFloat}}
      )
      EndColor=<rgba 0.8854 0.8854 0.8854 1>
      LerpFactor=&MakeCustomFloatFromBool(Bool=$User.bIsFrontEndPreview, True Float=1, False Float=0)
     ))) {isSetParameters=true}
   /SetVariables(Particles.RimLightDynamic:Vector4f=&MakeVector4_FromLinearColor(LinearColor=&LerpLinearColors(
      StartColor=&ColorFromCurve(
       ColorCurve=DataInterface() {propertyValues {
         AlphaCurve {keys [<at 0 1>,<at 0.125 1>,<at 0.25 1>,<at 0.375 1>,<at 0.5 1>,<at 0.625 1>,<at 0.75 1>,<at 0.875 1>,<at 1 1>]}
         BlueCurve {keys [<at 0 0.01>,<at 0.041 0.01>,<at 0.083 0.01>,<at 0.125 0.01>,<at 0.166 0.0077>,<at 0.208 0.0045>,<at 0.25 0.0092>,<at 0.291 0.0136>,<at 0.333 0.03>,<at 0.375 0.04>,<at 0.416 0.0525>,<at 0.458 0.0525>,<at 0.5 0.0525>,<at 0.541 0.0525>,<at 0.583 0.0525>,<at 0.625 0.04>,<at 0.666 0.03>,<at 0.708 0.0152>,<at 0.75 0.0153>,<at 0.791 0.0079>,<at 0.833 0.0025>,<at 0.875 0.0031>,<at 0.916 0.01>,<at 0.958 0.01>,<at 1 0.01>]}
         GreenCurve {keys [<at 0 0.0068>,<at 0.041 0.0064>,<at 0.083 0.0064>,<at 0.125 0.0071>,<at 0.166 0.0056>,<at 0.208 0.0041>,<at 0.25 0.0103>,<at 0.291 0.0186>,<at 0.333 0.0156>,<at 0.375 0.0208>,<at 0.416 0.0272>,<at 0.458 0.0272>,<at 0.5 0.0272>,<at 0.541 0.0272>,<at 0.583 0.0272>,<at 0.625 0.0208>,<at 0.666 0.0156>,<at 0.708 0.0192>,<at 0.75 0.0153>,<at 0.791 0.0092>,<at 0.833 0.0051>,<at 0.875 0.0062>,<at 0.916 0.0064>,<at 0.958 0.0063>,<at 1 0.0068>]}
         RedCurve {keys [<at 0 0.0033>,<at 0.041 0.0029>,<at 0.083 0.0032>,<at 0.125 0.0036>,<at 0.166 0.01>,<at 0.208 0.01>,<at 0.25 0.02>,<at 0.291 0.03>,<at 0.333 0.0086>,<at 0.375 0.0115>,<at 0.416 0.0151>,<at 0.458 0.0151>,<at 0.5 0.0151>,<at 0.541 0.0151>,<at 0.583 0.0151>,<at 0.625 0.0115>,<at 0.666 0.0086>,<at 0.708 0.03>,<at 0.75 0.03>,<at 0.791 0.02>,<at 0.833 0.01>,<at 0.875 0.01>,<at 0.916 0.0022>,<at 0.958 0.0019>,<at 1 0.0033>]}
        }}
       CurveIndex=Linked() {linkedVariable {name=NPC.FortniteNPC.TimeOfDay_0-1, type=NiagaraFloat}}
      )
      EndColor=<rgba 0.0151 0.0272 0.0525 1>
      LerpFactor=&MakeCustomFloatFromBool(Bool=$User.bIsFrontEndPreview, True Float=1, False Float=0)
     ))) {isSetParameters=true}
   &DynamicMaterialParameters(
    Index 0 Param 3=&Lerp_Float(
     A=&FloatFromCurve(
      FloatCurve=DataInterface() {propertyValues {Curve {keys [
          <at 0.216 1 RCIM_Cubic>
          {interpMode=RCIM_Cubic, tangentMode=RCTM_User, time=0.3505, value=0}
          {interpMode=RCIM_Cubic, tangentMode=RCTM_User, time=0.71, value=0}
          <at 0.8425 1 RCIM_Cubic>
         ]}}}
      CurveIndex=Linked() {linkedVariable {name=NPC.FortniteNPC.TimeOfDay_0-1, type=NiagaraFloat}}
     )
     B=0
     Alpha=&MakeCustomFloatFromBool(Bool=$User.bIsFrontEndPreview, True Float=1, False Float=0)
    )
    Index 0 Param 4=&Lerp_Float(
     A=&FloatFromCurve(
      FloatCurve=DataInterface() {propertyValues {Curve {keys [
          <at 0.216 0.4 RCIM_Cubic>
          {interpMode=RCIM_Cubic, tangentMode=RCTM_User, time=0.3505, value=0.94}
          {interpMode=RCIM_Cubic, tangentMode=RCTM_User, time=0.71, value=0.94}
          <at 0.8425 0.4 RCIM_Cubic>
         ]}}}
      CurveIndex=Linked() {linkedVariable {name=NPC.FortniteNPC.TimeOfDay_0-1, type=NiagaraFloat}}
     )
     B=1
     Alpha=&MakeCustomFloatFromBool(Bool=$User.bIsFrontEndPreview, True Float=1, False Float=0)
    )
    Index 1 Param 1=&MakeCustomFloatFromBool(Bool=$User.bIsFrontEndPreview, True Float=0.4, False Float=0.9)
    Index 1 Param 2=&MakeCustomFloatFromBool(Bool=$User.bIsFrontEndPreview, True Float=0, False Float=1)
    Module.Index0Param0Write=true
    Module.Index0Param1Write=true
    Module.Index0Param2Write=true
    Module.Index0Param3Write=true
    Module.Index1Param0Write=true
    Module.Index1Param1Write=true
    Module.Index1Param2Write=true
    Module.Index1Param3Write=true
    Write Parameter Index 1=true
   ) {module=DynamicMaterialParameters001}
  ]
  rendererClasses [&NiagaraRibbonRendererProperties]
  renderers [{rendererClass=&NiagaraRibbonRendererProperties}]
 }
}
summary {
 systemName=NS_Shoe_AbyssBlot_Idle
 emitters [
  {emitterName=FN_AnimTrail, rendererClasses [&NiagaraRibbonRendererProperties]}
  {emitterName=FN_AnimTrail003, rendererClasses [&NiagaraRibbonRendererProperties]}
 ]
 userVariables [
  User.bIsFrontEnd {type=NiagaraBool}
  User.bIsFrontEndPreview {type=NiagaraBool}
  User.bIsSwinging {type=NiagaraBool}
  User.HideFootFX {type=NiagaraBool}
  User.IsSkydiving {type=NiagaraBool}
  User.RibbonIDOffset {type=NiagaraInt32}
  User.ShoeScale_L {type=Vector3f, description="", defaultValue {struct=Vector3f, value=<1 1 1>}}
  User.ShoeScale_R {type=Vector3f, description="", defaultValue {struct=Vector3f, value=<1 1 1>}}
 ]
}
