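# Character cosmetic material for the AbyssBlot shoes, adding animated, music-reactive rainbow emissive with eye and pattern masks.
# Layers tech-art color, emissive and roughness over the shared character, shoe and clear-coat functions, then applies sock, PostFX and hologram functions.
# Emissive samples a gradient ramp offset by pre-skinned position, Fresnel, time and masks.
# Audio amplitude or an off-character sine supplies reactivity; platform switches simplify flowmap distortion and noise on Switch and mobile.
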Material /CosmeticShoes/Assets/FORT_Shoes/Shoe_AbyssBlot/Materials/M_AbyssBlot_Shoes
graph {
 properties {
  bUseMaterialAttributes=true
  bUsedWithClothing=true
  bUsedWithMeshDeformer=true
  bUsedWithMorphTargets=true
  bUsedWithNanite=true
  bUsedWithNiagaraMeshParticles=true
  bUsedWithSkeletalMesh=true
  pixelDepthOffsetMode=PDOM_Legacy
  shadingModel=MSM_FromMaterialExpression
  usedShadingModels=Default Lit | Subsurface | Clear Coat | Subsurface Profile | Cloth
 }
 nodes [
  n1 = Multiply(A=Time(), B=ScalarParameter() {defaultValue=0.07, group=MusicReactivity_Ctrl, parameterName=OffCharacter_Speed})
  n2 = SmoothStep(Value=ConstantBiasScale(None=Sine(None=@n1))) {constMin=0.6}
  LinearInterpolate(
   A=MaterialFunctionCall() {materialFunction=/Game/Packages/Fortress_SharedMaterials/Base_Material_Functions/BaseCharacterShaderFunctions/Audio/MF_AudioReactivity}.Amplitude
   B=@n2
   Alpha=ScalarParameter() {defaultValue=1, group=MusicReactivity_Ctrl, parameterName=IsInFrontEnd}
  )
  n3 = MaterialFunctionCall() {materialFunction=&MF_CharacterParent_2023_PreFX}
  n4 = OneMinus(None=NamedRerouteUsage())
  n5 = NamedRerouteUsage()
  n6 = GetMaterialAttributes(None=MaterialFunctionCall!(
    Material=MaterialFunctionCall!(Material=@n3.Result) {materialFunction=/Game/Global/Materials/Cosmetics/Kicks/Shoes/Materials/MF_ShoeReactivityTest}.Result
    Alpha=LinearInterpolate(A=@n3.Tech Art Mask, B=@n4, Alpha=@n5) {constB=0}
   ) {materialFunction=&/MF_ClearCoat_TechArt_RE}.Result) {attributeGetTypes [#1, #2, #3]}
  n7 = NamedRerouteUsage()
  n8 = ScalarParameter() {defaultValue=0.3, group=MusicReactivity, parameterName=BaseColorOverride}
  n9 = SmoothStep(Value=NamedRerouteUsage()) {constMax=0.3}
  n10 = Fresnel() {baseReflectFraction=0, exponent=3}
  n11 = NamedRerouteUsage()
  n12 = NamedRerouteUsage()
  n13 = Multiply(
   A=@n7
   B=LinearInterpolate(A=LinearInterpolate(Alpha=@n10) {constA=0.4}, B=SmoothStep(Value=@n11) {constMax=0.6}, Alpha=Power(Base=@n12) {constExponent=4})
  )
  n14 = NamedRerouteUsage()
  n15 = SetMaterialAttributes!(
   MaterialAttributes=@n6.MaterialAttributes
   Base Color=LinearInterpolate(A=@n6.Base Color, B=Multiply(A=@n7, B=@n8) {constB=0.2}, Alpha=@n9)
   Emissive Color=Add(A=@n6.Emissive Color, B=Multiply(A=@n13, B=@n9))
   Roughness=LinearInterpolate(A=@n6.Roughness, B=LinearInterpolate(Alpha=@n14) {constA=1, constB=0.2}, Alpha=@n9)
  ) {attributeSetTypes [#2, #1, #3]}
  MaterialFunctionCall!(In=MaterialFunctionCall(
    Material=MaterialFunctionCall!(InOpacity, Material=@n15, Mask=@n3.Tech Art Mask) {materialFunction=/Game/Global/Materials/Cosmetics/Kicks/Socks/Materials/MF_Sock}.Result
    M=@n3.M
    Pre Skinned Local Position=@n3.Pre skinned Local Position
    Distance Blend=@n3.Pre skinned Local Position
    Fresnel=@n3.Pre skinned Local Position
   ) {materialFunction=&MF_CharacterParent_2023_PostFX}.Result) {materialFunction=/Game/Global/Materials/Cosmetics/Kicks/Shoes/Materials/MF_TechartShoeHologram}
  n16 = TextureSample(Tex=NamedRerouteUsage()) {
   samplerSource=SSM_Wrap_WorldGroupSettings
   samplerType=SAMPLERTYPE_LinearColor
   texture=/CosmeticShoes/Assets/FORT_Shoes/Shoe_AbyssBlot/Textures/T_Abyss_TechArt
  }
  NamedRerouteDeclaration(None=@n16.R) {name=EyeMask, nodeColor=<rgba 1 0.5859 0 1>}
  NamedRerouteDeclaration(None=Multiply(A=@n16.G, B=NamedRerouteUsage())) {name=AreaMask, nodeColor=<rgba 0.2656 0 1 1>}
  n17 = MaterialFunctionCall!(InAmplitude, InSpectrum, CustomFakeSpectrum, CustomFakeAmplitude) {materialFunction=/BRCosmetics/Global/MaterialFunctions/AudioReactivity/MF_AudioReactivity_FrontendLoop}
  n18 = ScalarParameter() {defaultValue=1, group=MusicReactivity_Ctrl, parameterName=IsOnCharacter}
  NamedRerouteDeclaration(None=Saturate(None=LinearInterpolate(A=@n2, B=@n17.Amplitude, Alpha=@n18))) {name=ReactivityAlpha, nodeColor=<rgba 1 0.2344 0 1>}
  n19 = NamedRerouteUsage()
  n20 = Add(A=Multiply(A=Time(), B=ScalarParameter() {defaultValue=0.7, group=MusicReactivity, parameterName=PatternSpeed}), B=NamedRerouteUsage())
  n21 = LinearInterpolate(Alpha=ConstantBiasScale(None=Sine(None=@n20))) {constA=0.4, constB=0.6}
  n22 = NamedRerouteUsage()
  n23 = Add(A=Multiply(A=Time(), B=ScalarParameter() {defaultValue=-0.5, group=MusicReactivity, parameterName=PatternSpeed02}), B=NamedRerouteUsage())
  n24 = LinearInterpolate(Alpha=ConstantBiasScale(None=Sine(None=@n23))) {constA=0.7, constB=0.9}
  n25 = Add(
   A=Multiply(A=SphereMask(A=@n19, B=@n21) {attenuationRadius=0.1, hardnessPercent=1}, B=SmoothStep(Value=@n19) {constMax=0.1})
   B=SphereMask(A=@n22, B=@n24) {attenuationRadius=0.1, hardnessPercent=1}
  )
  NamedRerouteDeclaration(None=Saturate(None=@n25)) {name=Pattern01, nodeColor=<rgba 0 0.1797 1 1>}
  n26 = Step(Y=@n21, X=@n19)
  n27 = Step(Y=@n22, X=@n24)
  NamedRerouteDeclaration(None=Multiply(A=OneMinus(None=@n25), B=Multiply(A=@n26, B=@n27))) {name=CenterMask, nodeColor=<rgba 0.5937 1 0 1>}
  n28 = TextureObjectParameter() {
   group=MusicReactivity
   parameterName=TechArtMask_RE
   samplerType=SAMPLERTYPE_LinearColor
   texture=&T_Abyss_TechArt
  }
  NamedRerouteDeclaration(None=@n28) {name=TechArtRE, nodeColor=<rgba 1 0.6563 0 1>}
  n29 = TextureCoordinate()
  n30 = Multiply(A=TextureCoordinate(), B=ScalarParameter() {defaultValue=3, group=MusicReactivity, parameterName=DistortionTiling01})
  n31 = VectorParameter() {group=MusicReactivity, parameterName=DistortionSpeed, defaultValue=<rgba 0.03 0.5 0.05 0.2>}
  n32 = TextureObjectParameter() {group=MusicReactivity, parameterName=DistortionTexture, texture=/Game/Global/Textures/Default/Flowmaps/T_SwirlFlow_03}
  n33 = TextureSample(UVs=Panner(Coordinate=@n30, Speed=AppendVector(A=@n31.R, B=@n31.R)), Tex=@n32) {samplerSource=SSM_Wrap_WorldGroupSettings, samplerType=SAMPLERTYPE_Normal, texture=/Game/Global/Textures/Default/Flowmaps/T_SwirlFlow_02_N}
  n34 = OneMinus(None=NamedRerouteUsage())
  n35 = Multiply(
   A=Subtract(A=ComponentMask(None=@n33.RGB) {g=true, r=true}) {constB=0.5}
   B=Multiply(A=SmoothStep(Value=@n34) {constMin=0.2}, B=ScalarParameter() {defaultValue=0.05, group=MusicReactivity, parameterName=DistortionIntensity01})
  )
  n36 = Constant()
  n37 = TextureSample(
   UVs=Add(
    A=@n29
    B=MaterialFunctionCall!("Xb1,PS4"=@n35, "PC-High&&Epic,XSX,PS5"=@n35, "PC-Perf,Android"=@n36, Switch=@n36, "PC-Low(NotPerf)&&Medium"=@n35) {materialFunction=/Game/Effects/Niagara/MasterMaterials/Functions/Common/PlatformSwitch_2D}.Result
   )
   Tex=NamedRerouteUsage()
  ) {
   samplerSource=SSM_Wrap_WorldGroupSettings
   samplerType=SAMPLERTYPE_LinearColor
   texture=&T_Abyss_TechArt
  }
  NamedRerouteDeclaration(None=@n37.B) {name=TechArtGradient01, nodeColor=<rgba 0.3359 0 1 1>}
  n38 = TextureCoordinate()
  n39 = ScalarParameter() {defaultValue=3, group=MusicReactivity, parameterName=NoiseTiling01}
  n40 = VectorParameter() {group=MusicReactivity, parameterName=NoiseSpeed, defaultValue=<rgba 0.2 0.2 0 1>}
  n41 = TextureSample(
   UVs=Panner(Coordinate=Multiply(A=@n38, B=@n39), Speed=AppendVector(A=@n40.R, B=@n40.R))
   Tex=NamedRerouteDeclaration(None=TextureObjectParameter() {group=MusicReactivity, parameterName=NoiseMask, texture=/Game/Global/Textures/Default/Noise/T_SoftNoise}) {name=NoiseMask, nodeColor=<rgba 0.0312 1 0 1>}
  ) {samplerSource=SSM_Wrap_WorldGroupSettings, samplerType=SAMPLERTYPE_Normal, texture=&T_SwirlFlow_02_N}
  n42 = Constant()
  NamedRerouteDeclaration(None=MaterialFunctionCall("Xb1,PS4"=@n41.R, "PC-High&&Epic,XSX,PS5"=@n41.R, "PC-Perf,Android"=@n42, Switch=@n42, "PC-Low(NotPerf)&&Medium"=@n41.R) {materialFunction=&PlatformSwitch_1D}.Result) {name=TimeNoise01, nodeColor=<rgba 1 0 0.3516 1>}
  n43 = TextureSample(UVs=Panner(Coordinate=@n30, Speed=AppendVector(A=@n31.B, B=@n31.B)), Tex=@n32) {samplerSource=SSM_Wrap_WorldGroupSettings, samplerType=SAMPLERTYPE_Normal, texture=&T_SwirlFlow_02_N}
  n44 = Subtract(A=ComponentMask(None=@n43.RGB) {g=true, r=true}) {constB=0.5}
  n45 = NamedRerouteUsage()
  n46 = Multiply(
   A=SmoothStep(Value=OneMinus(None=@n45)) {constMin=0.2}
   B=ScalarParameter() {defaultValue=0.02, group=MusicReactivity, parameterName=DistortionIntensity02}
  )
  n47 = TextureSample(UVs=Add(A=@n29, B=Multiply(A=@n44, B=@n46)), Tex=NamedRerouteUsage()) {
   samplerSource=SSM_Wrap_WorldGroupSettings
   samplerType=SAMPLERTYPE_LinearColor
   texture=&T_Abyss_TechArt
  }
  n48 = NamedRerouteUsage()
  NamedRerouteDeclaration(None=MaterialFunctionCall("Xb1,PS4"=@n47.B, "PC-High&&Epic,XSX,PS5"=@n47.B, "PC-Perf,Android"=@n48, Switch=@n48, "PC-Low(NotPerf)&&Medium"=@n47.B) {materialFunction=&PlatformSwitch_2D}.Result) {name=TechArtGradient02, nodeColor=<rgba 0.3359 0 1 1>}
  NamedRerouteDeclaration(None=MaterialFunctionCall("Xb1,PS4"=@n41.G, "PC-High&&Epic,XSX,PS5"=@n41.G, "PC-Perf,Android"=@n42, Switch=@n42, "PC-Low(NotPerf)&&Medium"=@n41.G) {materialFunction=&PlatformSwitch_1D}.Result) {name=TimeNoise02, nodeColor=<rgba 1 0.1641 0 1>}
  NamedRerouteDeclaration(None=MaterialFunctionCall("Xb1,PS4"=@n41.B, "PC-High&&Epic,XSX,PS5"=@n41.B, "PC-Perf,Android"=@n42, Switch=@n42, "PC-Low(NotPerf)&&Medium"=@n41.B) {materialFunction=&PlatformSwitch_1D}.Result) {name=TimeNoise03, nodeColor=<rgba 0 0.9297 1 1>}
  NamedRerouteDeclaration(None=@n16.B) {name=PatternGradient, nodeColor=<rgba 0.0547 1 0 1>}
  n49 = TextureSampleParameter2D() {
   group=MusicReactivity
   parameterName=TechArtMask_Add
   texture=/CosmeticShoes/Assets/FORT_Shoes/Shoe_AbyssBlot/Textures/T_AbyssBlot_Shoe_TechArt_Add
  }
  NamedRerouteDeclaration(None=@n49.B) {name=DotMask, nodeColor=<rgba 1 0.75 0 1>}
  NamedRerouteDeclaration(None=ComponentMask(None=@n3.M) {g=true}) {name=Gmap, nodeColor=<rgba 0 0.4609 1 1>}
  n50 = Fresnel() {baseReflectFraction=0, exponent=1}
  n51 = ComponentMask(None=NamedRerouteUsage()) {g=true}
  n52 = ScalarParameter() {defaultValue=120, group=MusicReactivity, parameterName=PreskinnedGradientUnit01}
  n53 = NamedRerouteUsage()
  n54 = Add(
   A=Add(A=Multiply(A=@n50) {constB=0.1}, B=Divide(A=@n51, B=@n52))
   B=Multiply(A=@n53, B=ScalarParameter() {defaultValue=0.5, group=MusicReactivity, parameterName=EyeOffset})
  )
  n55 = NamedRerouteUsage()
  n56 = Saturate(None=Add(A=Multiply(A=@n53, B=@n55), B=NamedRerouteUsage()))
  n57 = NamedRerouteUsage()
  n58 = ScalarParameter() {defaultValue=0.6, group=MusicReactivity, parameterName=ColorOffset}
  n59 = Multiply(A=Saturate(None=Add(A=@n56, B=@n57)), B=@n58)
  n60 = Time()
  n61 = ScalarParameter() {defaultValue=0.4, group=MusicReactivity, parameterName=RainbowSpeed}
  n62 = Add(
   A=Add(A=Add(A=@n54, B=@n59), B=Multiply(A=@n60, B=@n61))
   B=Multiply(A=NamedRerouteUsage(), B=ScalarParameter() {defaultValue=0.16, group=MusicReactivity, parameterName=CenterOffset})
  )
  n63 = Multiply(A=Power(Base=NamedRerouteUsage()), B=ScalarParameter() {defaultValue=0.17, group=MusicReactivity, parameterName=ReactivityOffset}) {constB=0.6}
  n64 = TextureSampleParameter2D(UVs=AppendVector(A=Add(A=@n62, B=@n63), B=Constant() {r=0.5})) {group=MusicReactivity, parameterName=GradientRainbow, texture=/CosmeticShoes/Assets/FORT_Shoes/Shoe_AbyssBlot/Textures/T_Abyss_Gradient}
  n65 = NamedRerouteUsage()
  n66 = NamedRerouteUsage()
  n67 = LinearInterpolate(
   A=ScalarParameter() {defaultValue=1, group=MusicReactivity, parameterName=EM_Intensity}
   B=ScalarParameter() {defaultValue=0.8, group=MusicReactivity, parameterName=EM_Intensity_Eye}
   Alpha=Saturate(None=Subtract(A=@n65, B=@n66))
  )
  n68 = NamedRerouteUsage()
  n69 = Multiply(
   A=Multiply(A=Multiply(A=@n64.RGB, B=@n67), B=LinearInterpolate(Alpha=@n68) {constA=1, constB=0.7})
   B=LinearInterpolate(Alpha=NamedRerouteUsage()) {constA=1, constB=0.2}
  )
  n70 = Multiply(A=Power(Base=NamedRerouteUsage()) {constExponent=3}, B=NamedRerouteUsage())
  n71 = TextureSample() {texture=/CosmeticShoes/Assets/FORT_Shoes/Shoe_AbyssBlot/Textures/T_AbyssBlot_TechArtGradient}
  n72 = Constant() {r=1}
  n73 = MaterialFunctionCall("Xb1,PS4"=@n71.R, "PC-High&&Epic,XSX,PS5"=@n71.R, "PC-Perf,Android"=@n72, Switch=@n72, "PC-Low(NotPerf)&&Medium"=@n71.R) {materialFunction=&PlatformSwitch_1D}
  n74 = NamedRerouteUsage()
  n75 = LinearInterpolate(Alpha=Multiply(A=Multiply(A=@n70, B=@n73.Result), B=OneMinus(None=@n74))) {constA=1, constB=20}
  n76 = Power(Base=NamedRerouteUsage()) {constExponent=4}
  n77 = ComponentMask(None=NamedRerouteUsage()) {g=true}
  n78 = ScalarParameter() {defaultValue=20, group=MusicReactivity, parameterName=PreskinnedGradientUnit02}
  n79 = Time()
  n80 = ScalarParameter() {defaultValue=1, group=MusicReactivity, parameterName=ScrollingSpeed}
  n81 = Sine(None=Add(A=Divide(A=@n77, B=@n78), B=Multiply(A=@n79, B=@n80)))
  n82 = LinearInterpolate(A=LinearInterpolate(Alpha=SmoothStep(Value=@n76) {constMax=0.7}) {constA=0.5}, Alpha=Power(Base=ConstantBiasScale(None=@n81))) {constA=0.5, constB=1.2}
  n83 = NamedRerouteUsage()
  NamedRerouteDeclaration(None=Multiply(A=Multiply(A=@n69, B=@n75), B=LinearInterpolate(A=@n82, Alpha=@n83))) {name=EM_OUT, nodeColor=<rgba 0 1 0.7422 1>}
  NamedRerouteDeclaration(None=SmoothStep(Value=Fresnel() {baseReflectFraction=0}) {constMax=0.2, constMin=0.1}) {name=EyeFresnel, nodeColor=<rgba 0.7344 0 1 1>}
  NamedRerouteDeclaration(None=@n58) {name=ColorOffset, nodeColor=<rgba 0 0.3906 1 1>}
  NamedRerouteDeclaration(None=@n3.Pre skinned Local Position) {name=PreskinPos, nodeColor=<rgba 0.3828 0 1 1>}
  NamedRerouteDeclaration(None=SmoothStep(Value=@n49.R) {constMax=0.1}) {name=ConstantMask, nodeColor=<rgba 0.125 1 0 1>}
  NamedRerouteDeclaration(None=@n49.G) {name=DarkenMask, nodeColor=<rgba 1 0 0.4453 1>}
  NamedRerouteDeclaration(None=@n56) {name=EyeConstant, nodeColor=<rgba 1 0.5859 0 1>}
  NamedRerouteUsage()
  ScalarParameter() {defaultValue=1, group=MusicReactivity_Ctrl, parameterName=ReactivityAlpha}
 ]
}
